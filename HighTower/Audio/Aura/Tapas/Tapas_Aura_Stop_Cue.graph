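# Sound cue named as the stop sound for the Tapas aura in HighTower, currently empty with no root or sound nodes, so it has no audio to play.
# It still carries playback settings: raised volume, lowered pitch, a maximum distance and the third-person friendly player foley sound class.
# A concurrency override caps simultaneous instances and stops the quietest when the limit is exceeded.
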SoundCue /HighTower/Audio/Aura/Tapas/Tapas_Aura_Stop_Cue
root=null
nodes []
properties {
 bOverrideConcurrency=true
 maxDistance=4000
 pitchMultiplier=0.6
 soundClassObject=/Game/Sounds/FortSoundClassesAndModes/Fort_Player_Foley_3P_Friendly_Athena
 volumeMultiplier=1.5
 concurrencyOverrides {maxCount=3, resolutionRule=StopQuietest}
}
